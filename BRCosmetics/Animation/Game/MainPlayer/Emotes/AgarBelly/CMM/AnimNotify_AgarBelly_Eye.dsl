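; Anim notify state for the AgarBelly emote that toggles a head material effect, apparently for the eyes, over the notify window.
; Begin and End check that the pawn's head is the AgarBelly female mesh, gating changes through a variant local.
; Helpers set the "AgarbellyEmote" scalar to 1, then 0.
; They also override Face and Body slot 0 with base-less dynamic materials, then null; the created head instance is never applied.
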(fn Received_NotifyBegin
  (params (MeshComp :type Skeletal%20Mesh%20Component%20Object%20Reference) (Animation :type Anim%20Sequence%20Base%20Object%20Reference) (TotalDuration :type Float%20%28single-precision%29) (EventReference :type Anim%20Notify%20Event%20Reference%20Structure%20%28by%20ref%29))
  (bind _asfort_player_pawn (Utilities|Casting|CastToFortPlayerPawn (Components|GetOwner MeshComp)))
  (Utilities|FlowControl|Sequence
    (:then_0
      (Utilities|IsValid (Customization|GetSkeletalMeshforPartType _asfort_player_pawn "Head" :!class /Script/FortniteGame.FortPlayerPawn)
        (:Is%20Valid
          (if (Utilities|Equal(Object) "/BRCosmetics/Characters/Player/Female/Medium/Heads/F_MED_AgarBelly_Head/Meshes/F_MED_AgarBelly_Head.F_MED_AgarBelly_Head" (Variables|/Script/Engine.SkinnedMeshComponent|GetSkeletalMesh (Customization|GetSkeletalMeshforPartType _asfort_player_pawn "Head" :!class /Script/FortniteGame.FortPlayerPawn)))
            (LocalVariables|SetCurrent%20Progressive%20Variant 1)))))
    (:then_2
      (if (== (LocalVariables|GetCurrent%20Progressive%20Variant) 1)
        (CallFunction|SetPartMaterialsDynamic _asfort_player_pawn "/BRCosmetics/Characters/Player/Female/Medium/Bodies/F_MED_AgarBelly/Materials/MI_AgarBelly_Head.MI_AgarBelly_Head")))
    (:then_3
      (return false))))

(fn Received_NotifyEnd
  (params (MeshComp :type Skeletal%20Mesh%20Component%20Object%20Reference) (Animation :type Anim%20Sequence%20Base%20Object%20Reference) (EventReference :type Anim%20Notify%20Event%20Reference%20Structure%20%28by%20ref%29))
  (bind _asfort_player_pawn (Utilities|Casting|CastToFortPlayerPawn (Components|GetOwner MeshComp)))
  (Utilities|FlowControl|Sequence
    (:then_0
      (Utilities|IsValid (Customization|GetSkeletalMeshforPartType _asfort_player_pawn "Head" :!class /Script/FortniteGame.FortPlayerPawn)
        (:Is%20Valid
          (if (Utilities|Equal(Object) "/BRCosmetics/Characters/Player/Female/Medium/Heads/F_MED_AgarBelly_Head/Meshes/F_MED_AgarBelly_Head.F_MED_AgarBelly_Head" (Variables|/Script/Engine.SkinnedMeshComponent|GetSkeletalMesh (Customization|GetSkeletalMeshforPartType _asfort_player_pawn "Head" :!class /Script/FortniteGame.FortPlayerPawn)))
            (LocalVariables|SetCurrent%20Progressive%20Variant 1)))))
    (:then_2
      (if (== (LocalVariables|GetCurrent%20Progressive%20Variant) 1)
        (CallFunction|SetPartEndMaterials0 _asfort_player_pawn "/BRCosmetics/Characters/Player/Female/Medium/Bodies/F_MED_AgarBelly/Materials/MI_AgarBelly_Head.MI_AgarBelly_Head")))
    (:then_3
      (return false))))

(fn SetPartMaterials
  (params (Pawn :type Fort%20Player%20Pawn%20Object%20Reference) (Face%20Material :type Material%20Instance%20Object%20Reference) (Head%20Material :type Material%20Instance%20Object%20Reference) (Body%20Material :type Material%20Interface%20Object%20Reference))
  (Utilities|IsValid Pawn
    (:Is%20Valid
      (Utilities|FlowControl|Sequence
        (:then_0
          (Rendering|Material|SetMaterial (Customization|GetSkeletalMeshforPartType Pawn "Face" :!class /Script/FortniteGame.FortPlayerPawn) 0 Face%20Material))
        (:then_1
          (Rendering|Material|SetScalarParameterValueonMaterials (Customization|GetSkeletalMeshforPartType Pawn "Head" :!class /Script/FortniteGame.FortPlayerPawn) "AgarbellyEmote" 1.0))
        (:then_2
          (Rendering|Material|SetMaterial (Customization|GetSkeletalMeshforPartType Pawn "Body" :!class /Script/FortniteGame.FortPlayerPawn) 0 Body%20Material))))))

(fn SetPartMaterialsDynamic
  (params (Player%20Pawn :type Fort%20Player%20Pawn%20Object%20Reference) (Head%20Material%20Base :type Material%20Interface%20Object%20Reference))
  (bind _returnvalue (Rendering|Material|CreateDynamicMaterialInstance Player%20Pawn 0 "" "None" :!class /Script/Engine.KismetMaterialLibrary))
  (bind _returnvalue_2 (Rendering|Material|CreateDynamicMaterialInstance Player%20Pawn Head%20Material%20Base "" "None" :!class /Script/Engine.KismetMaterialLibrary))
  (bind _returnvalue_3 (Rendering|Material|CreateDynamicMaterialInstance Player%20Pawn 0 "" "None" :!class /Script/Engine.KismetMaterialLibrary))
  (CallFunction|SetPartMaterials Player%20Pawn _returnvalue _returnvalue_2 _returnvalue_3))

(fn SetPartEndMaterials_0
  (params (Pawn :type Fort%20Player%20Pawn%20Object%20Reference) (Head%20Material :type Material%20Instance%20Object%20Reference))
  (Utilities|IsValid Pawn
    (:Is%20Valid
      (Utilities|FlowControl|Sequence
        (:then_0
          (Rendering|Material|SetMaterial (Customization|GetSkeletalMeshforPartType Pawn "Face" :!class /Script/FortniteGame.FortPlayerPawn) 0 0))
        (:then_1
          (Rendering|Material|SetScalarParameterValueonMaterials (Customization|GetSkeletalMeshforPartType Pawn "Head" :!class /Script/FortniteGame.FortPlayerPawn) "AgarbellyEmote" 0.0))
        (:then_2
          (Rendering|Material|SetMaterial (Customization|GetSkeletalMeshforPartType Pawn "Body" :!class /Script/FortniteGame.FortPlayerPawn) 0 0))))))
